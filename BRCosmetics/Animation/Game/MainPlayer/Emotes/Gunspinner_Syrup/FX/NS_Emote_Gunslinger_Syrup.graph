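# Niagara system for the Gunslinger Syrup emote, drawing spinning highlights and a disk attached to a skeletal mesh.
# Both emitters spawn short-lived particles that resample position and rotation each update from the User Skeletal Mesh.
# CircularHighlights scales a mesh up while spinning it by a random total angle over its life, with spawn rate cut late by an age curve.
# NE_DarkDisk spawns steadily, orients its mesh to the sampled rotation and fades alpha out.
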NiagaraSystem /BRCosmetics/Animation/Game/MainPlayer/Emotes/Gunspinner_Syrup/FX/NS_Emote_Gunslinger_Syrup
dependencies {
 usedDataInterfaces [&NiagaraDataInterfaceCurve, &NiagaraDataInterfaceVectorCurve]
 usedDynamicInputs []
 usedModules [&SystemState, /BRCosmetics/Athena/Effects/Particles/Gliders/NS_Glider_CarbideBlack/Module/SkeletalMeshData]
}
emitters {
 CircularHighlights {
  emitterUpdateScript [
   &EmitterState(Inactive Response=::NewEnumerator1, Life Cycle Mode=::NewEnumerator1, Loop Behavior=::NewEnumerator1, Loop Duration=4, UseLoopDelay=false)
   &SpawnRate(
    SpawnRate=&Multiply_Float(
     A=10
     B=&FloatFromCurve(FloatCurve=DataInterface() {propertyValues {Curve {keys [<at 0 1>,<at 0.7477 0.9843>,<at 0.7775 0.0215>]}}}, CurveIndex=$Emitter.Age)
    )
    Use Spawn Probability=false
   )
  ]
  particleSpawnScript [
   /Niagara/Modules/Spawn/Initialization/InitializeParticle(
    Lifetime=0.1
    Mass=0
    Mesh Scale=&Multiply_VectorByFloat(Vector=<0.08 0.08 0.01>, Float=&UniformRangedFloat(Minimum=0.7, Maximum=1))
    Module.Write Color=true
    Module.Write Lifetime=true
    Module.Write Mass=true
    Module.Write Material Random=true
    Module.Write Position=false
    Module.Write Scale=true
    Module.Write SpriteRotation=false
    Module.Write SpriteSize=true
    Position=&SimulationPosition()
   )
   &Color()
   &DynamicMaterialParameters(
    Index 0 Param 1=&UniformRangedFloat(Minimum=0.7, Maximum=0.4)
    Index 0 Param 2=1
    Index 0 Param 3=1
    Index 0 Param 4=1
    Module.Param0WriteEnabled=true
    Module.Param1WriteEnabled=true
    Module.Param2WriteEnabled=true
    Module.Param3WriteEnabled=true
   )
   &ApplyInitialForces()
   /SetVariables(Particles.GnomonScale:Vector3f=&Multiply_VectorByFloat(Vector=<1 1 1>, Float=1)) {isSetParameters=true}
   /SetVariables(Particles.RotationAmount:NiagaraFloat=&UniformRangedFloat(Minimum=0, Maximum=50000)) {isSetParameters=true}
   &CameraOffset(Camera Offset Amount=2)
  ]
  particleUpdateScript [
   &ParticleState()
   &SolveForcesAndVelocity(Clamp Velocity=false, Limit Acceleration=false)
   &SkeletalMeshLocation(Bone Sampling Mode=::NewEnumerator4, Inherit Velocity=false, MaskSpawnGroup=false, Skeletal Mesh=$User.Skeletal Mesh)
   &ScaleMeshSize(Scale Factor=&VectorFromCurve(VectorCurve=DataInterface() {propertyValues {XCurve {keys [<at 0 0.7>,<at 1 1>]}, YCurve {keys [<at 0 0.7>,<at 1 1>]}, ZCurve {keys [<at 0 1>,<at 1 1>]}}}))
   /SetVariables(Particles.MeshOrientation=&MultiplyQuaternion(
     Quaternion A=&MultiplyQuaternion(Quaternion A=$Output.SkeletalMeshLocation.SampledRotation, Quaternion B=<0 1 0 1>)
     Quaternion B=&MakeQuatFromAxisAngle(Axis=<0 0 1>, AngleInDegrees=&Multiply_Float(A=$Particles.RotationAmount, B=$Particles.NormalizedAge))
    )) {isSetParameters=true}
  ]
  rendererClasses [&NiagaraSpriteRendererProperties, &NiagaraMeshRendererProperties]
  renderers [
   {rendererClass=&NiagaraSpriteRendererProperties}
   {rendererClass=&NiagaraMeshRendererProperties, rendererIndex=1}
   {rendererClass=&NiagaraMeshRendererProperties, rendererIndex=2}
  ]
 }
 NE_DarkDisk {
  emitterUpdateScript [
   &EmitterState()
   &SpawnRate(SpawnRate=20, Use Spawn Probability=false)
   &SpawnBurst_Instantaneous(Spawn Count=1, Use Spawn Probability=false) {enabled=false}
  ]
  particleSpawnScript [
   &InitializeParticle(
    Lifetime=0.1
    Mesh Scale=<0.25 0.2 0.2>
    Module.Write Color=true
    Module.Write Lifetime=true
    Module.Write Mass=true
    Module.Write Material Random=true
    Module.Write Position=false
    Module.Write Scale=true
    Module.Write SpriteRotation=false
    Module.Write SpriteSize=true
    Position=&SimulationPosition()
    Sprite Size=<5 5>
   )
   /Niagara/Modules/Spawn/Orientation/V2/InitialMeshOrientation(Use Orientation Vector=true, Use Rotation Vector=false) {enabled=false}
   &Color()
  ]
  particleUpdateScript [
   &ParticleState()
   &SkeletalMeshLocation(Bone Sampling Mode=::NewEnumerator4, Inherit Velocity=false, MaskSpawnGroup=false, Skeletal Mesh=$User.Skeletal Mesh)
   &ScaleColor(
    Scale Alpha=&FloatFromCurve(FloatCurve=DataInterface() {propertyValues {Curve {keys [<at 0 1 RCIM_Cubic>,<at 1 0>]}}})
    ScaleA=true
    ScaleRGB=true
   )
   /SetVariables(Particles.MeshOrientation=&MultiplyQuaternion(Quaternion A=$Output.SkeletalMeshLocation.SampledRotation, Quaternion B=<1 0 0 1>)) {isSetParameters=true}
  ]
  rendererClasses [&NiagaraSpriteRendererProperties, &NiagaraMeshRendererProperties]
  renderers [
   {rendererClass=&NiagaraSpriteRendererProperties}
   {rendererClass=&NiagaraMeshRendererProperties, rendererIndex=1}
  ]
 }
}
summary {
 systemName=NS_Emote_Gunslinger_Syrup
 emitters [
  {emitterName=NE_DarkDisk, rendererClasses [&NiagaraSpriteRendererProperties, &NiagaraMeshRendererProperties]}
  {
   emitterName=CircularHighlights
   rendererClasses [&NiagaraSpriteRendererProperties, &NiagaraMeshRendererProperties]
  }
 ]
 userVariables [User.Skeletal Mesh {
   type=NiagaraDataInterfaceSkeletalMesh
   defaultValue {
    struct=NiagaraExt_VariableValue_DataInterface
    value {dataInterface=&self:NiagaraDataInterfaceSkeletalMesh_0}
   }
  }]
}
